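# Unlit material for an animated star screen on Sparks stages, colored by the stage's primary and secondary colors.
# Crossfades two scaled samples of a packed star texture into a loop, alongside a pulsing center star, noisy dots, radiating bands, frame and vignette masks.
# UVs scale, rotate and support horizontal flip via Custom Primitive Data.
# A curve-remapped audio beat supplies reactivity, hue shift and MF_Sparks_Screen_Details finish the look, and debug switches substitute colors and beat.
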Material /SparksMaps/Effects/Screens/Star/M_Sparks_Screen_Star
graph {
 properties {bAutomaticallySetUsageInEditor=false, bHasPixelAnimation=true, shadingModel=MSM_Unlit, usedShadingModels=Unlit}
 nodes [
  n1 = ScalarParameter() {group=Color, parameterName=Hueshift_Global_Speed}
  n2 = Add(
   A=Add(A=MaterialFunctionCall(Speed=@n1) {materialFunction=&DebugTimeSine}.Result) {constB=0.5}
   B=ScalarParameter() {group=Color, parameterName=Hueshift_Offset}
  )
  n3 = NamedRerouteUsage()
  n4 = NamedRerouteUsage()
  n5 = Multiply(A=Add(A=Add(A=@n3, B=@n4), B=NamedRerouteUsage()), B=NamedRerouteUsage())
  MaterialFunctionCall!(
   Emissive=Multiply(
    A=MaterialFunctionCall(Hue Shift Percentage=@n2, Texture=@n5) {materialFunction=&HueShift}.Result
    B=NamedRerouteUsage()
   )
   Scanlines Tile
  ) {materialFunction=/SparksMaps/Effects/Screens/Materials/MF/MF_Sparks_Screen_Details}
  n6 = VectorParameter() {group=Debug -- set back to FALSE, parameterName=Debug_SecondaryColor, defaultValue=<rgba 0.0615 0.4774 0 1>}
  n7 = MaterialFunctionCall!(CustomUPosition) {materialFunction=/SparksStageShow/Systems/StageManager/FXManager/MaterialFunctions/MF_MojoStageColors}
  n8 = StaticBoolParameter() {group=Debug -- set back to FALSE, parameterName=isDebug?}
  NamedRerouteDeclaration(None=NamedRerouteDeclaration(None=StaticSwitch(True=@n6.RGB, False=@n7.Secondary, Value=@n8)) {name=SecondaryColor, nodeColor=<rgba 0 0 0 1>}) {name=ColorB, nodeColor=<rgba 1 0.75 0 1>}
  n9 = ScalarParameter() {group=Color, parameterName=HueshiftBG_Min}
  n10 = ScalarParameter() {defaultValue=0.2, group=Color, parameterName=HueshiftBG_Max}
  n11 = OneMinus(None=NamedRerouteUsage())
  n12 = Constant() {r=0.005}
  n13 = Multiply(A=MaterialFunctionCall(Speed=Constant() {r=0.06}) {materialFunction=&DebugTimeSine}.Result) {constB=2}
  n14 = MaterialFunctionCall(UVs=NamedRerouteUsage(), Texture Scale=Saturate(None=LinearInterpolate(Alpha=@n13) {constA=0.42})) {materialFunction=&ScaleUVsByCenter}
  n15 = LinearInterpolate(
   A=ScalarParameter() {defaultValue=0.35, parameterName=Shading}
   Alpha=Add(
    A=MaterialFunctionCall(Speed=@n12) {materialFunction=&DebugTimeSine}.Result
    B=TextureSample!(UVs=@n14.UVs, Tex, Level, Apply View MipBias) {mipValueMode=TMVM_MipLevel, samplerType=SAMPLERTYPE_LinearColor, texture=&/T_Screen_Star_Packed}.B
   )
  ) {constA=0.25}
  n16 = Multiply(
   A=MaterialFunctionCall(Hue Shift Percentage=LinearInterpolate(A=@n9, B=@n10, Alpha=@n11) {constB=0.2}, Texture=NamedRerouteUsage()) {materialFunction=&HueShift}.Result
   B=@n15
  )
  n17 = Multiply(A=NamedRerouteUsage(), B=NamedRerouteUsage())
  n18 = NamedRerouteUsage()
  n19 = LinearInterpolate(Alpha=NamedRerouteUsage()) {constA=1, constB=3}
  n20 = NamedRerouteUsage()
  n21 = LinearInterpolate(Alpha=@n20) {constA=1, constB=3}
  n22 = Sine(None=Frac(None=Add(A=@n20) {constB=-0.25}))
  n23 = LinearInterpolate(
   A=TextureSample(UVs=MaterialFunctionCall(UVs=@n18, Texture Scale=@n19) {materialFunction=&ScaleUVsByCenter}.UVs) {mipValueMode=TMVM_MipLevel, samplerType=SAMPLERTYPE_LinearColor, texture=&T_Screen_Star_Packed}.G
   B=TextureSample(UVs=MaterialFunctionCall(UVs=@n18, Texture Scale=@n21) {materialFunction=&ScaleUVsByCenter}.UVs) {mipValueMode=TMVM_MipLevel, samplerType=SAMPLERTYPE_LinearColor, texture=&T_Screen_Star_Packed}.G
   Alpha=Power(Base=ConstantBiasScale(None=@n22)) {constExponent=12}
  )
  NamedRerouteDeclaration(None=Multiply(A=LinearInterpolate(A=@n16, B=@n17, Alpha=@n23), B=@n15)) {name=StarLoop, nodeColor=<rgba 0.5469 1 0 1>}
  n24 = TextureCoordinate()
  n25 = LinearInterpolate(
   A=ScalarParameter() {defaultValue=1, group=UVs, parameterName=UV_SineScale_Min}
   B=ScalarParameter() {defaultValue=2, group=UVs, parameterName=UV_SineScale_Max}
   Alpha=MaterialFunctionCall(Speed=Constant() {r=0.025}) {materialFunction=&DebugTimeSine}.Result
  ) {constA=1, constB=2}
  n26 = Multiply(A=NamedRerouteUsage(), B=ComponentMask(None=@n25) {r=true})
  n27 = ComponentMask(None=@n25) {g=true}
  n28 = AppendVector(
   A=ScalarParameter() {defaultValue=1, group=UVs, parameterName=UV_Scale_U}
   B=ScalarParameter() {defaultValue=1, group=UVs, parameterName=UV_Scale_V}
  )
  n29 = MaterialFunctionCall(UVs=@n24, Texture Scale=Multiply(A=@n28, B=AppendVector(A=@n26, B=@n27))) {materialFunction=&ScaleUVsByCenter}
  n30 = AppendVector(A=ScalarParameter() {group=UVs, parameterName=UV_Offset_X}, B=ScalarParameter() {group=UVs, parameterName=UV_Offset_Y})
  n31 = Time()
  n32 = Multiply(A=NamedRerouteUsage(), B=ScalarParameter() {defaultValue=0.05, group=UVs, parameterName=UVRotate_Speed})
  n33 = ConstantBiasScale(None=Sine(None=Multiply(A=@n31, B=@n32)))
  NamedRerouteDeclaration(None=Rotator(Coordinate=Add(A=@n29.UVs, B=@n30), Time=LinearInterpolate(Alpha=@n33) {constA=-1})) {name=UVs_Modified, nodeColor=<rgba 1 0 0.0001 1>}
  n34 = NamedRerouteUsage()
  n35 = Power(Base=TextureSample!(UVs=Panner(Coordinate=@n34) {speedX=0.03, speedY=0.12}, Tex, Apply View MipBias) {samplerType=SAMPLERTYPE_Grayscale, texture=&LowResBlurredNoise}.R)
  n36 = Multiply(A=Rotator(Coordinate=@n34) {speed=0.02}, B=Constant() {r=85}) {constB=100}
  n37 = Multiply(A=SphereMask(A=Frac(None=@n36), B=Constant() {r=0.5}, Radius=Constant() {r=0.2}) {attenuationRadius=0.1, hardnessPercent=1}) {constB=2}
  NamedRerouteDeclaration(None=Multiply(A=Multiply(A=@n35, B=@n37), B=ScalarParameter() {defaultValue=0.2, group=Color, parameterName=Dots_Intensity})) {name=Dots, nodeColor=<rgba 1 0.1406 0 1>}
  n38 = ScalarParameter() {defaultValue=0.2, group=Color, parameterName=Frame_Amount}
  n39 = ComponentMask(None=TextureCoordinate()) {g=true}
  n40 = OneMinus(None=@n38)
  NamedRerouteDeclaration(None=Multiply(A=Step(Y=@n38, X=@n39) {constX=0.2, constY=0.2}, B=Step(Y=@n39, X=@n40) {constX=0.8, constY=0.2})) {name=Frame, nodeColor=<rgba 0 0.7422 1 1>}
  n41 = LinearInterpolate(Alpha=NamedRerouteUsage()) {constA=0.62, constB=0.6}
  n42 = ScalarParameter() {group=Color, parameterName=VignetteRadius_Offset}
  NamedRerouteDeclaration(None=MaterialFunctionCall(UVs=NamedRerouteUsage(), Radius=Add(A=@n41, B=@n42)) {materialFunction=&RadialGradientExponential}.RadialGradientExponential) {name=Vignette, nodeColor=<rgba 0 1 0.7656 1>}
  n43 = MaterialFunctionCall(UVs=NamedRerouteUsage(), Texture Scale=LinearInterpolate(Alpha=NamedRerouteUsage()) {constA=0.16, constB=0.3}) {materialFunction=&ScaleUVsByCenter}
  NamedRerouteDeclaration(None=Multiply(A=TextureSample!(UVs=@n43.UVs, Tex, Level, Apply View MipBias) {mipValueMode=TMVM_MipLevel, samplerType=SAMPLERTYPE_LinearColor, texture=&T_Screen_Star_Packed}.R) {constB=0.666}) {name=CenterPulseStar, nodeColor=<rgba 0 1 0.8359 1>}
  n44 = Time()
  n45 = ScalarParameter() {defaultValue=0.2, parameterName=Speed}
  NamedRerouteDeclaration(None=Frac(None=Multiply(A=@n44, B=@n45))) {name=Time01, nodeColor=<rgba 0.0547 1 0 1>}
  n46 = Multiply(A=@n45, B=Time())
  NamedRerouteDeclaration(None=Frac(None=Add(A=@n46) {constB=0.5})) {name=Time02, nodeColor=<rgba 1 0 0.375 1>}
  n47 = Constant() {r=0.6}
  n48 = Constant() {r=-2}
  n49 = Constant() {r=2}
  n50 = Constant() {r=1}
  n51 = Frac(None=Panner(Coordinate=MaterialFunctionCall(Radius=@n50) {materialFunction=&RadialGradientExponential}.RadialGradientExponential) {speedY=0.1})
  NamedRerouteDeclaration(None=LinearInterpolate(Alpha=MaterialFunctionCall!(Width=@n47, Sharpness=@n48, Offset, Direction Switch, Compare, Bands=@n49, Input Coordinates=@n51) {materialFunction=/Engine/Functions/Engine_MaterialFunctions03/Procedurals/GeneratedOffsetBands}.Result) {constA=0.5, constB=3}) {name=RadiatingBands, nodeColor=<rgba 0.1016 0 1 1>}
  n52 = Saturate(None=ScalarParameter() {bUseCustomPrimitiveData=true, parameterName="CPD_HorizontalFlip? (0 or 1)", primitiveDataIndex=4})
  NamedRerouteDeclaration(None=LinearInterpolate(Alpha=Round(None=@n52)) {constA=1, constB=-1}) {name=CDP_UVHorizontalFlip, nodeColor=<rgba 0 1 0.1281 1>}
  n53 = Multiply(A=AppendVector(A=NamedRerouteUsage(), B=Constant() {r=1}), B=@n28)
  NamedRerouteDeclaration(None=Add(
    A=MaterialFunctionCall(UVs=@n24, Texture Scale=@n53) {materialFunction=&ScaleUVsByCenter}.UVs
    B=@n30
   )) {name=UVs_NoSineScale, nodeColor=<rgba 0.2066 0.0042 0 1>}
  n54 = Time()
  n55 = OneMinus(None=Frac(None=Multiply(A=@n54) {constB=2}))
  n56 = CurveAtlasRowParameter(CurveTime=CollectionParameter() {collection=/FMAudioReactiveData/AudioAnalysis/MPC_Sparks_Audio_Parameters, parameterName=BeatLifetime}) {
   atlas=/SparksMaps/Effects/Screens/Curves/CurveAtlas_Audio_Beat_A
   curve=/SparksMaps/Effects/Screens/Curves/Curve_Audio_Beat_Remap_001
   parameterName=ReactivityCurve
  }
  n57 = Power(Base=@n56, Exp=ScalarParameter() {defaultValue=3, parameterName=BeatPow})
  n58 = Constant()
  n59 = Constant() {r=1}
  n60 = Constant() {r=1}
  n61 = Constant() {r=2}
  NamedRerouteDeclaration(None=StaticSwitch(
    True=Multiply(A=@n55, B=@n55)
    False=MaterialFunctionCall(Input=@n57, Input Low=@n58, Input High=@n59, Target Low=@n60, Target High=@n61) {materialFunction=&RemapValueRange}.Result
    Value=StaticBoolParameter() {group=Debug -- set back to FALSE, parameterName=isDebug?}
   )) {name=Reactivity, nodeColor=<rgba 0 0 0 1>}
  n62 = VectorParameter() {group=Debug -- set back to FALSE, parameterName=Debug_PrimaryColor, defaultValue=<rgba 0.6094 0.2865 0 1>}
  NamedRerouteDeclaration(None=NamedRerouteDeclaration(None=StaticSwitch(True=@n62.RGB, False=@n7.Primary, Value=@n8)) {name=PrimaryColor, nodeColor=<rgba 0 0 0 1>}) {name=ColorA, nodeColor=<rgba 0 0.9297 1 1>}
 ]
}
